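# Translucent unlit two-sided effect material for the Apparition Grunt's Deimos skin chain, drawing glowing, spline-thickened strands.
# World position offset sums SplineThicken width, sine-masked wobble from a scrolling world-space normal texture, and a vertex-blue-masked sine hover.
# Emissive derives HDR color from an LDR tint, with an optional time pulse.
# Opacity is a sine falloff along V; vertex green blends width, color and depth fade, while vertex red varies motion timing.
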Material /Game/Characters/Player/Male/Medium/Bodies/M_Med_ApparitionGrunt/FX/M_Deimos_Apparition_Skin_Chain
graph {
 properties {
  bUsedWithMeshDeformer=true
  bUsedWithNiagaraMeshParticles=true
  bUsedWithNiagaraRibbons=true
  bUsedWithNiagaraSprites=true
  bUsedWithSkeletalMesh=true
  blendMode=BLEND_Translucent
  pixelDepthOffsetMode=PDOM_Legacy
  shadingModel=MSM_Unlit
  twoSided=true
  usedShadingModels=Unlit
 }
 nodes [
  n1 = WorldPosition()
  n2 = ObjectPositionWS()
  n3 = Time()
  n4 = VertexColor()
  n5 = Add(
   A=Divide(A=Subtract(A=@n1.XYZ, B=@n2), B=ScalarParameter() {defaultValue=12, parameterName=WPO_Tile})
   B=Multiply(A=Add(A=@n3, B=@n4.R), B=ScalarParameter() {defaultValue=0.4, parameterName=WPO_Speed}) {constB=0.5}
  )
  n6 = Multiply(
   A=VertexNormalWS()
   B=TextureSample(UVs=ComponentMask(None=@n5) {g=true, r=true}) {samplerType=SAMPLERTYPE_Normal, texture=/Game/Effects/Fort_Effects/Gameplay/SmokeDustExplosions/Textures/T_Explosion_NM_01}.RGB
  )
  n7 = Multiply(A=MaterialFunctionCall(Float2=TextureCoordinate()) {materialFunction=&BreakOutFloat2Components}.R) {constB=0.5}
  n8 = Multiply(A=Saturate(None=Sine(None=@n7)), B=ScalarParameter() {defaultValue=1, parameterName=WPO_Intensity})
  n9 = ScalarParameter() {defaultValue=0.5, parameterName=Width}
  n10 = LinearInterpolate(A=@n9, B=Multiply(A=@n9) {constB=15}, Alpha=VertexColor().G)
  n11 = StaticBool()
  n12 = WorldPosition()
  n13 = ObjectPositionWS()
  n14 = ScalarParameter() {defaultValue=1, parameterName=Time_offsetSpread_01}
  n15 = Time()
  n16 = ScalarParameter() {defaultValue=0.2, parameterName=Speed_01}
  n17 = VertexColor()
  n18 = Sine(None=Multiply(A=Add(A=@n15, B=@n14), B=Multiply(A=@n16, B=@n17.R)))
  n19 = ScalarParameter() {defaultValue=1.5, parameterName=Time_offsetSpread_02}
  n20 = ScalarParameter() {defaultValue=0.2, parameterName=Speed_02}
  n21 = Sine(None=Multiply(A=Add(A=@n15, B=@n19), B=Multiply(A=@n20, B=@n17.R)))
  n22 = ScalarParameter() {defaultValue=2, parameterName=Time_offsetSpread_03}
  n23 = ScalarParameter() {defaultValue=0.2, parameterName=Speed_03}
  n24 = Sine(None=Multiply(A=Add(A=@n15, B=@n22), B=Multiply(A=@n23, B=@n17.R)))
  n25 = Add(
   A=Normalize(VectorInput=Subtract(A=@n12.XYZ, B=@n13))
   B=Multiply(
    A=MaterialFunctionCall(X=@n18, Y=@n21, Z=@n24) {materialFunction=&MakeFloat3}.Result
    B=VectorParameter() {parameterName=WPO_Offset_Amt, defaultValue=<rgba 2 2 2 1>}.RGB
   )
  )
  Add(
   A=Add(
    A=Multiply(A=@n6, B=@n8)
    B=MaterialFunctionCall(WidthBase=@n10, WidthTip=@n10, Expand U or V UV Channel=@n11) {materialFunction=&SplineThicken}.WorldPositionOffset
   )
   B=Multiply(A=Multiply(A=@n25, B=@n17.B), B=ScalarParameter() {defaultValue=1, parameterName=UseDeimosChunkHoverWPO})
  )
  n26 = VectorParameter() {parameterName=Color, defaultValue=<rgba 0.3004 0.0917 0.4896 1>}
  n27 = ScalarParameter() {defaultValue=10, parameterName=Brightness}
  n28 = LinearInterpolate(
   A=MaterialFunctionCall(LDR Input=@n26.RGB, HDR Dynamic Range=@n27, HDR Tint=Multiply(A=@n26.RGB, B=@n27), HDR Intensity Multiplier=@n27) {materialFunction=&DeriveHDRfromLDR}.RemappedRange
   B=VectorParameter() {parameterName=Color, defaultValue=<rgba 0.3004 0.0917 0.4896 1>}.RGB
   Alpha=VertexColor().G
  )
  n29 = Divide(A=@n28) {constB=15}
  n30 = Time()
  n31 = ConstantBiasScale(None=Sine(None=Divide(A=@n30)))
  MaterialFunctionCall(Emissive=LinearInterpolate(A=@n28, B=LinearInterpolate(A=@n29, B=@n28, Alpha=@n31), Alpha=ScalarParameter() {parameterName=UseEmissivePulse})) {materialFunction=&MF_FX_EmissiveAdjustments}
  n32 = Multiply(A=MaterialFunctionCall(Float2=TextureCoordinate()) {materialFunction=&BreakOutFloat2Components}.G) {constB=0.5}
  n33 = Power(Base=Saturate(None=Sine(None=@n32))) {constExponent=3}
  n34 = Divide(A=@n33) {constB=4}
  n35 = Constant() {r=20}
  Multiply(
   A=LinearInterpolate(A=@n33, B=DepthFade(Opacity=@n34, FadeDistance=@n35), Alpha=VertexColor().G)
   B=ScalarParameter() {defaultValue=1, parameterName=BeamOpacity}
  )
 ]
}
