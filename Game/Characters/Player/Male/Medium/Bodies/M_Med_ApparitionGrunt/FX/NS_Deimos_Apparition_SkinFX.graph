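# Niagara skin effect for the Deimos Apparition grunt player body, adding chest glows and optional smoke.
# ChestGlow001 bursts two sprite-and-mesh particles, glued each frame to system-level skeletal mesh samples with rotated offsets, optionally sine-pulsed through NeedPulsing.
# SmokeSubUV emits darkened, animated sprites from another sample, gated by SpawnSmoke probability (default zero).
# User parameters set color and scale for both emitters; per-particle SkeletalMeshLocation modules are disabled.
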NiagaraSystem /Game/Characters/Player/Male/Medium/Bodies/M_Med_ApparitionGrunt/FX/NS_Deimos_Apparition_SkinFX
dependencies {
 usedDataInterfaces [
  &NiagaraDataInterfaceArrayFloat
  &NiagaraDataInterfaceArrayInt32
  &NiagaraDataInterfaceArrayFloat3
  &NiagaraDataInterfaceCurve
 ]
 usedDynamicInputs []
 usedModules [&SystemState, &SampleSkeletalMesh]
}
emitters {
 ChestGlow001 {
  emitterUpdateScript [
   &EmitterState(
    Inactive Response=::NewEnumerator1
    Life Cycle Mode=::NewEnumerator1
    Loop Behavior=::NewEnumerator1
    Loop Duration Mode=::NewEnumerator1
    UseLoopDelay=false
   )
   &SpawnBurst_Instantaneous(Spawn Count=2, Use Spawn Probability=false, UseLoopCountLimit=false)
  ]
  particleSpawnScript [
   &FN_SetSpawnGroup(Number of Spawn Groups=2)
   &InitializeParticle(
    Color=$User.FX_Color
    Mesh Scale Mode=::NewEnumerator3
    Mesh Uniform Scale=&Multiply_Float(A=$User.FX_Scalar, B=0.22)
    Sprite Size=&Multiply_Vector2D_ByFloat(Float=$User.FX_Scalar, Vector2D=<60 60>)
    Sprite Size Mode=::NewEnumerator4
    UsePositionOffset=false
    Write Material Random=true
   )
   /SetVariables() {isSetParameters=true}
   &InitialMeshOrientation(Mesh Axis To Orient=<0 0 1>, Mesh Orientation Mode=::NewEnumerator2, Use Rotation Vector=false)
   &CameraOffset(Camera Offset Amount=&SelectFloatFromArray(Direct Array Index=$Particles.SpawnGroup, Float Selection Array=DataInterface() {propertyValues {FloatData [5,8]}}))
  ]
  particleUpdateScript [
   &ParticleState(Kill Particles When Lifetime Has Elapsed=false, Let Infinitely Lived Particles Die When Emitter Deactivates=true)
   &SkeletalMeshLocation(
    Bone / Socket Index=&SelectIntFromArray(Direct Array Index=$Particles.SpawnGroup, Int Selection Array=DataInterface() {propertyValues {IntData [0,1]}})
    Bone Sampling Mode=::NewEnumerator4
    MaskSpawnGroup=false
    Sampled Position Offset=&SelectVectorFromArray(Direct Array Index=$Particles.SpawnGroup, Vector Selection Array=DataInterface() {propertyValues {FloatData [<1 5 0>, <1 -11 0>]}})
    Skeletal Mesh=$User.SK_Mesh
   ) {enabled=false}
   /SetVariables(Particles.Position=&MakeCustomVectorFromBool(
     Bool=&SetBoolByIntComparison(A=$Particles.SpawnGroup, B=0, Comparison Type=::NewEnumerator4)
     True Vector=&Add_Vector(
      A=$System.SampleSkeletalMesh001.SampledPositionWithoutOffset
      B=&RotateVectorByQuaternion(Vector To Rotate=<1 5 0>, Quaternion=$System.SampleSkeletalMesh001.SampledRotation)
     )
     False Vector=&Add_Vector(
      A=$System.SampleSkeletalMesh002.SampledPositionWithoutOffset
      B=&RotateVectorByQuaternion(Vector To Rotate=<1 -11 0>, Quaternion=$System.SampleSkeletalMesh002.SampledRotation)
     )
    )) {isSetParameters=true}
   /SetVariables(Particles.MeshOrientation=/Niagara/DynamicInputs/TypeConversions/MakeCustomQuaternionFromBool(
     Bool:NiagaraBool=&SetBoolByIntComparison(A=$Particles.SpawnGroup, B=0, Comparison Type=::NewEnumerator4)
     True Quat:Quat4f=$System.SampleSkeletalMesh001.SampledRotation
     False Quat:Quat4f=$System.SampleSkeletalMesh002.SampledRotation
    )) {isSetParameters=true}
   &ScaleColor(
    Scale RGB=&VectorFromFloat(Value=&MakeCustomFloatFromBool(
      Bool=$User.NeedPulsing
      True Float=&Sine(Normalized Angle=$Engine.Time, Period=2, Scale=0.5, Bias=0.5)
      False Float=1
     ))
    ScaleA=true
    ScaleRGB=true
   )
  ]
  rendererClasses [&NiagaraSpriteRendererProperties, &NiagaraMeshRendererProperties]
  renderers [
   {rendererClass=&NiagaraSpriteRendererProperties}
   {rendererClass=&NiagaraMeshRendererProperties, rendererIndex=1}
  ]
 }
 SmokeSubUV {
  emitterUpdateScript [
   &EmitterState()
   &SpawnRate(Spawn Probability=$User.SpawnSmoke, SpawnRate=5, Use Spawn Probability=true)
  ]
  particleSpawnScript [
   &InitializeParticle(
    Color=$User.FX_Color
    Lifetime Max=3
    Lifetime Mode=::NewEnumerator1
    Sprite Rotation Mode=::NewEnumerator2
    Sprite Size Mode=::NewEnumerator3
    Uniform Sprite Size=&Multiply_Float(A=$User.FX_Scalar, B=30)
    UsePositionOffset=false
    Write Material Random=true
   )
   &SkeletalMeshLocation(Bone / Socket Index=1, Bone Sampling Mode=::NewEnumerator4, MaskSpawnGroup=false, Skeletal Mesh=$User.SK_Mesh) {enabled=false}
   /SetVariables(Particles.Position:Vector3f=$System.SampleSkeletalMesh.SampledPositionWithoutOffset) {isSetParameters=true}
   &AddVelocity(Velocity=<0 -20 50>) {enabled=false}
   &AddVelocityInCone(Cone Angle=35, Cone Axis=<0 -0.9 1>, Module.Use Velocity Falloff On Cone Axis=false, Velocity Strength=50)
   &CameraOffset(Camera Offset Amount=5)
  ]
  particleUpdateScript [
   &InheritVelocity(Speed Limit=true, UseSpeedThreshold=false) {enabled=false}
   &Drag(Use Linear Drag=true, Use Rotational Drag=false) {enabled=false}
   &InheritSourceMovement(
    Applied Force Scale=<-1 -1 -1>
    ApplyForce=true
    ApplyForceFalloff=false
    ApplyPosition=true
    ApplyPositionFalloff=false
    Module.LimitForceSpeed=false
   ) {enabled=false}
   &SolveForcesAndVelocity(Clamp Velocity=false, Limit Acceleration=false)
   &ParticleState()
   &SpriteRotationRate(Rotation Rate=&RandomRangeFloat(Minimum=-50, Maximum=50))
   &SubUVAnimation(
    End Frame=45
    Start Frame=15
    SubUV Animation Mode:ENiagara_SubUVLookupMode=::NewEnumerator1
    SubUV Curve=DataInterface() {propertyValues {Curve {keys [{arriveTangent=1.2926, interpMode=RCIM_Cubic, leaveTangent=1.2926, tangentMode=RCTM_User, time=0, value=0}, <at 1 1 RCIM_Cubic>]}}}
   )
   &ScaleColor(
    Scale Alpha=&FloatFromCurve(FloatCurve=DataInterface() {propertyValues {Curve {keys [
         <at 0 0 RCIM_Cubic>
         <at 0.0906 1.5>
         {arriveTangent=-0.0405, interpMode=RCIM_Cubic, leaveTangent=-0.0405, tangentMode=RCTM_User, time=0.458, value=1}
         {arriveTangent=-3.2044, interpMode=RCIM_Cubic, leaveTangent=-3.2044, tangentMode=RCTM_User, time=1, value=0}
        ]}}})
    Scale RGB=&VectorFromFloat(Value=0.1)
    ScaleA=true
    ScaleRGB=true
   )
  ]
  rendererClasses [&NiagaraSpriteRendererProperties]
  renderers [{rendererClass=&NiagaraSpriteRendererProperties}]
 }
}
summary {
 systemName=NS_Deimos_Apparition_SkinFX
 emitters [
  {emitterName=ChestGlow001, rendererClasses [&NiagaraSpriteRendererProperties, &NiagaraMeshRendererProperties]}
  {emitterName=SmokeSubUV, rendererClasses [&NiagaraSpriteRendererProperties]}
 ]
 userVariables [
  User.FX_Color {type=LinearColor, description="", defaultValue {struct=LinearColor, value=<rgba 50 8.6437 45.6627 0.1>}}
  User.FX_Scalar {type=NiagaraFloat, defaultValue=1, description=""}
  User.NeedPulsing {type=NiagaraBool}
  User.NewDeimosIndex {type=NiagaraInt32}
  User.SK_Mesh {
   type=NiagaraDataInterfaceSkeletalMesh
   defaultValue {
    struct=NiagaraExt_VariableValue_DataInterface
    value {dataInterface=&self:NiagaraDataInterfaceSkeletalMesh_0}
   }
  }
  User.SpawnSmoke {type=NiagaraFloat, defaultValue=0, description=""}
 ]
}
